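# Conversation database for the Punk Koi NPC, providing its greeting dialogue and access to shared NPC conversations.
# The entry point leads to one speech node with a default line and contextual variants for time since last match, first conversation with the NPC, and holding Bars.
# Curve rows set each variant's priority.
# The speech node then branches through link nodes to the shared NPCLibrary quests and basic services conversations.
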ConversationDatabase /NPCLibrary/NPCs/PunkKoi/PunkKoi/NPCLibrary_Conversation_PunkKoi
entry_points [{tag=Conversation.Entry.PunkKoi, destinations [#1]}]
nodes [
 ConversationEntryPointNode_7 {
  id=#1
  class=ConversationEntryPointNode
  outputs [#2]
  properties {entryTag {tagName=Conversation.Entry.PunkKoi}}
 }
 Speech_C_4 {
  id=#2
  class=Speech_C
  outputs [#3, #4]
  properties {
   nodeName=""
   generalConfig {
    defaultMessage="Are you gonna be cool, or are we gonna have a problem?"
    contextualMessages [
     {
      message=I've been picking up the slack since you've been gone. It was... a lot of slack.
      contextRequirements [{requirement {class=FortControllerRequirement_CheckTimeSinceLastPlayedMatch}}]
      priority {curve {rowName=Tandem.Speech.Priority.WhileSincePlayed}}
     }
     {
      message="What, you never seen a teenager possessed by a demon before? "
      contextRequirements [{requirement {class=FortControllerRequirement_IsFirstEverConversationWithNPC}}]
      priority {curve {rowName=Tandem.Speech.Priority.FirstEverInteraction}}
     }
     {
      message="Meh. Everyone thinks they're a bigshot when they have lotsa Bars. "
      contextRequirements [{requirement {class=FortControllerRequirement_HasBars}}]
      priority {curve {rowName=Tandem.Speech.Priority.HighTotalBars}}
     }
    ]
   }
  }
 }
 ConversationLinkNode_0 {id=#4, class=ConversationLinkNode, properties {remoteEntryTag {tagName=Conversation.Entry.BasicServices}}}
 ConversationLinkNode_1 {id=#3, class=ConversationLinkNode, properties {remoteEntryTag {tagName=Conversation.Entry.NPCLibrary.Quests}}}
]
